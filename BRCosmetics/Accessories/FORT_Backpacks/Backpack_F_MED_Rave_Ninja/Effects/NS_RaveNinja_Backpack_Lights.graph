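# Niagara cosmetic effect that lights the Rave Ninja backpack with zero-size sprites and light renderers.
# Eight single-burst emitters, two per socket, follow the big and small left/right sockets, tinted by LargeColor or SmallColor; small ones spawn at 50% probability.
# One set pulses brightness by age curve, boosted for front-end preview; the "001" set loops particle lifetime, scaled by LightMultiplier.
# Pink emitter names are side-swapped relative to their sockets.
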NiagaraSystem /BRCosmetics/Accessories/FORT_Backpacks/Backpack_F_MED_Rave_Ninja/Effects/NS_RaveNinja_Backpack_Lights
dependencies {
 usedDataInterfaces [&NiagaraDataInterfaceSkeletalMesh, &NiagaraDataInterfaceVectorCurve]
 usedDynamicInputs []
 usedModules [
  &self:SystemSpawnScript.SystemScriptSource.SystemScriptGraph.NiagaraNodeAssignment_1.SetVariables_#1
  &self:SystemSpawnScript.SystemScriptSource.SystemScriptGraph.NiagaraNodeAssignment_0.SetVariables_#2
  &SystemState
 ]
}
emitters {
 LT_Yellow {
  emitterUpdateScript [
   &EmitterState(
    Life Cycle Mode=::NewEnumerator1
    Loop Duration=&UniformRangedFloat(Minimum=0.5, Maximum=2)
    UseLoopDelay=false
   )
   &SpawnBurst_Instantaneous(Spawn Count=1, Spawn Probability=0.5, Use Spawn Probability=true)
   &SpawnRate(Use Spawn Probability=false) {enabled=false}
  ]
  particleSpawnScript [/Niagara/Modules/Spawn/Initialization/InitializeParticle(
    Color=&MultiplyLinearColor(
     Linear Color=$User.SmallColor
     Scale Factor=/Niagara/DynamicInputs/TypeConversions/MakeCustomVector4FromBool(Bool:NiagaraBool=$User.bIsFrontEndPreview, True Vector4:Vector4f=<7 7 7 1>, False Vector4:Vector4f=<1 1 1 1>)
    )
    Lifetime=$Emitter.CurrentLoopDuration
    Mass=&UniformRangedFloat(Minimum=0.4, Maximum=2)
    Module.Write Color=true
    Module.Write Lifetime=true
    Module.Write Mass=false
    Module.Write Material Random=true
    Module.Write Position=false
    Module.Write Scale=false
    Module.Write SpriteRotation=true
    Module.Write SpriteSize=true
    Position=&SimulationPosition()
    Sprite Rotation=&NormalizedAngleToDegrees(NormalizedAngle=&UniformRangedFloat(Minimum=0, Maximum=1))
    Sprite Size=<0 0>
   )]
  particleUpdateScript [
   &ParticleState()
   &SolveForcesAndVelocity(Clamp Velocity=false, Limit Acceleration=false)
   &SampleSkeletalMeshSkeleton(
    Mesh=DataInterface() {propertyValues {
      DefaultMesh=null
      ExcludeBoneName=root
      OptionalComponentClass=null
      PreviewMesh=/BRCosmetics/Accessories/FORT_Backpacks/Backpack_F_MED_Rave_Ninja/Meshes/F_MED_Rave_Ninja_Pack
      SkinningMode=SkinOnTheFly
      SoftSourceActor=null
      SourceMode=Default
      UvSetIndex=0
      WholeMeshLOD=-1
      bExcludeBone=false
      bReadDeformedGeometry=true
      bRequireCurrentFrameData=true
      ComponentTags []
      FilteredBones []
      FilteredSockets [LT_Small]
      MeshUserParameter {parameter {name=None, typeDefHandle {registeredTypeIndex=1}, varData []}}
      SamplingRegions []
     }}
    Spawn Mode:ENiagaraSkeletalSpawning=::NewEnumerator1
   )
   /Niagara/Modules/Spawn/Location/SkeletalMeshSkeletonLocation(Module.Inherit Velocity=false, Module.Use Sampled Bone as Ribbon ID=true)
   &ScaleColor(
    Scale RGB=&VectorFromFloat(Value=&MakeFloatFromVector(
      Vector=/Niagara/DynamicInputs/ScaleByCurve/ScaleVectorByCurve(
       InitialValue:Vector3f=<1 1 1>
       ScaleCurve:NiagaraDataInterfaceVectorCurve=DataInterface() {propertyValues {
         XCurve {keys [
           {arriveTangent=2, interpMode=RCIM_Cubic, time=0, value=0}
           <at 0.5 1.3 RCIM_Cubic>
           {interpMode=RCIM_Cubic, leaveTangent=-2, time=1, value=0}
          ]}
         YCurve {keys [<at 0.5458 0 RCIM_Cubic>,<at 0.8822 0 RCIM_Cubic>,<at 1.2639 0 RCIM_Cubic>]}
         ZCurve {keys [<at 0.5458 0 RCIM_Cubic>,<at 0.8822 0 RCIM_Cubic>,<at 1.2639 0 RCIM_Cubic>]}
        }}
       CurveIndex:NiagaraFloat=$Particles.NormalizedAge
      )
      Channel=::NewEnumerator0
     ))
    ScaleA=true
    ScaleRGB=true
   )
  ]
  rendererClasses [&NiagaraSpriteRendererProperties, &NiagaraLightRendererProperties]
  renderers [
   {rendererClass=&NiagaraSpriteRendererProperties}
   {rendererClass=&NiagaraLightRendererProperties, rendererIndex=1}
  ]
 }
 LT_Yellow001 {
  emitterUpdateScript [
   &EmitterState(Life Cycle Mode=::NewEnumerator1, Loop Behavior=::NewEnumerator1, Loop Duration Mode=::NewEnumerator1, UseLoopDelay=false)
   &SpawnBurst_Instantaneous(Spawn Count=1, Spawn Probability=0.5, Use Spawn Probability=true)
   &SpawnRate(Use Spawn Probability=false) {enabled=false}
  ]
  particleSpawnScript [&InitializeParticle(
    Color=$User.SmallColor
    Lifetime=$Emitter.CurrentLoopDuration
    Mass=&UniformRangedFloat(Minimum=0.4, Maximum=2)
    Module.Write Color=true
    Module.Write Lifetime=true
    Module.Write Mass=false
    Module.Write Material Random=true
    Module.Write Position=false
    Module.Write Scale=false
    Module.Write SpriteRotation=true
    Module.Write SpriteSize=true
    Position=&SimulationPosition()
    Sprite Rotation=&NormalizedAngleToDegrees(NormalizedAngle=&UniformRangedFloat(Minimum=0, Maximum=1))
    Sprite Size=<0 0>
   )]
  particleUpdateScript [
   &ParticleState(Kill Particles When Lifetime Has Elapsed=false, Loop Particles Lifetime=true)
   &SolveForcesAndVelocity(Clamp Velocity=false, Limit Acceleration=false)
   &SampleSkeletalMeshSkeleton(
    Mesh=DataInterface() {propertyValues {
      DefaultMesh=null
      ExcludeBoneName=root
      OptionalComponentClass=null
      PreviewMesh=&F_MED_Rave_Ninja_Pack
      SkinningMode=SkinOnTheFly
      SoftSourceActor=null
      SourceMode=Default
      UvSetIndex=0
      WholeMeshLOD=-1
      bExcludeBone=false
      bReadDeformedGeometry=true
      bRequireCurrentFrameData=true
      ComponentTags []
      FilteredBones []
      FilteredSockets [LT_Small]
      MeshUserParameter {parameter {name=None, typeDefHandle {registeredTypeIndex=1}, varData []}}
      SamplingRegions []
     }}
    Spawn Mode:ENiagaraSkeletalSpawning=::NewEnumerator1
   )
   &SkeletalMeshSkeletonLocation(Module.Inherit Velocity=false, Module.Use Sampled Bone as Ribbon ID=true)
   &ScaleColor(Scale RGB=&Multiply_VectorByFloat(Vector=<3 3 3>, Float=$User.LightMultiplier), ScaleA=true, ScaleRGB=true)
  ]
  rendererClasses [&NiagaraSpriteRendererProperties, &NiagaraLightRendererProperties]
  renderers [
   {rendererClass=&NiagaraSpriteRendererProperties}
   {rendererClass=&NiagaraLightRendererProperties, rendererIndex=1}
  ]
 }
 LT_pink {
  emitterUpdateScript [
   &EmitterState(Life Cycle Mode=::NewEnumerator1, Loop Duration=$System.StaticLifetime, UseLoopDelay=false)
   &SpawnBurst_Instantaneous(Spawn Count=1, Use Spawn Probability=false)
   &SpawnRate(Use Spawn Probability=false) {enabled=false}
  ]
  particleSpawnScript [&InitializeParticle(
    Color=&MultiplyLinearColor(
     Linear Color=$User.LargeColor
     Scale Factor=&MakeCustomVector4FromBool(Bool:NiagaraBool=$User.bIsFrontEndPreview, True Vector4:Vector4f=<7 7 7 1>, False Vector4:Vector4f=<1 1 1 1>)
    )
    Lifetime=&Multiply_Float(A=$Emitter.CurrentLoopDuration, B=1)
    Mass=&UniformRangedFloat(Minimum=0.4, Maximum=2)
    Module.Write Color=true
    Module.Write Lifetime=true
    Module.Write Mass=false
    Module.Write Material Random=true
    Module.Write Position=false
    Module.Write Scale=false
    Module.Write SpriteRotation=true
    Module.Write SpriteSize=true
    Position=&SimulationPosition()
    Sprite Rotation=&NormalizedAngleToDegrees(NormalizedAngle=&UniformRangedFloat(Minimum=0, Maximum=1))
    Sprite Size=<0 0>
   )]
  particleUpdateScript [
   &ParticleState()
   &SolveForcesAndVelocity(Clamp Velocity=false, Limit Acceleration=false)
   &SampleSkeletalMeshSkeleton(
    Mesh=DataInterface() {propertyValues {
      DefaultMesh=null
      ExcludeBoneName=root
      OptionalComponentClass=null
      PreviewMesh=&F_MED_Rave_Ninja_Pack
      SkinningMode=SkinOnTheFly
      SoftSourceActor=null
      SourceMode=Default
      UvSetIndex=0
      WholeMeshLOD=-1
      bExcludeBone=false
      bReadDeformedGeometry=true
      bRequireCurrentFrameData=true
      ComponentTags []
      FilteredBones []
      FilteredSockets [RT_Big]
      MeshUserParameter {parameter {name=None, typeDefHandle {registeredTypeIndex=1}, varData []}}
      SamplingRegions []
     }}
    Spawn Mode:ENiagaraSkeletalSpawning=::NewEnumerator1
   )
   &SkeletalMeshSkeletonLocation(Module.Inherit Velocity=false, Module.Use Sampled Bone as Ribbon ID=true)
   &ScaleColor(
    Scale RGB=&VectorFromFloat(Value=&MakeFloatFromVector(
      Vector=&ScaleVectorByCurve(
       InitialValue:Vector3f=<1 1 1>
       ScaleCurve:NiagaraDataInterfaceVectorCurve=DataInterface() {propertyValues {
         XCurve {keys [
           {arriveTangent=2, interpMode=RCIM_Cubic, time=0, value=0.3}
           <at 0.5 1.3 RCIM_Cubic>
           {interpMode=RCIM_Cubic, leaveTangent=-2, time=1, value=0.3}
          ]}
         YCurve {keys [<at 0.5458 0 RCIM_Cubic>,<at 0.8822 0 RCIM_Cubic>,<at 1.2639 0 RCIM_Cubic>]}
         ZCurve {keys [<at 0.5458 0 RCIM_Cubic>,<at 0.8822 0 RCIM_Cubic>,<at 1.2639 0 RCIM_Cubic>]}
        }}
       CurveIndex:NiagaraFloat=$Particles.NormalizedAge
      )
      Channel=::NewEnumerator0
     ))
    ScaleA=true
    ScaleRGB=true
   )
  ]
  rendererClasses [&NiagaraSpriteRendererProperties, &NiagaraLightRendererProperties]
  renderers [
   {rendererClass=&NiagaraSpriteRendererProperties}
   {rendererClass=&NiagaraLightRendererProperties, rendererIndex=1}
  ]
 }
 LT_pink001 {
  emitterUpdateScript [
   &EmitterState(Life Cycle Mode=::NewEnumerator1, Loop Behavior=::NewEnumerator1, Loop Duration Mode=::NewEnumerator1, UseLoopDelay=false)
   &SpawnBurst_Instantaneous(Spawn Count=1, Use Spawn Probability=false)
   &SpawnRate(Use Spawn Probability=false) {enabled=false}
  ]
  particleSpawnScript [&InitializeParticle(
    Color=$User.LargeColor
    Lifetime=&Multiply_Float(A=$Emitter.CurrentLoopDuration, B=1)
    Mass=&UniformRangedFloat(Minimum=0.4, Maximum=2)
    Module.Write Color=true
    Module.Write Lifetime=true
    Module.Write Mass=false
    Module.Write Material Random=true
    Module.Write Position=false
    Module.Write Scale=false
    Module.Write SpriteRotation=true
    Module.Write SpriteSize=true
    Position=&SimulationPosition()
    Sprite Rotation=&NormalizedAngleToDegrees(NormalizedAngle=&UniformRangedFloat(Minimum=0, Maximum=1))
    Sprite Size=<0 0>
   )]
  particleUpdateScript [
   &ParticleState(Kill Particles When Lifetime Has Elapsed=false, Loop Particles Lifetime=true)
   &SolveForcesAndVelocity(Clamp Velocity=false, Limit Acceleration=false)
   &SampleSkeletalMeshSkeleton(
    Mesh=DataInterface() {propertyValues {
      DefaultMesh=null
      ExcludeBoneName=root
      OptionalComponentClass=null
      PreviewMesh=&F_MED_Rave_Ninja_Pack
      SkinningMode=SkinOnTheFly
      SoftSourceActor=null
      SourceMode=Default
      UvSetIndex=0
      WholeMeshLOD=-1
      bExcludeBone=false
      bReadDeformedGeometry=true
      bRequireCurrentFrameData=true
      ComponentTags []
      FilteredBones []
      FilteredSockets [RT_Big]
      MeshUserParameter {parameter {name=None, typeDefHandle {registeredTypeIndex=1}, varData []}}
      SamplingRegions []
     }}
    Spawn Mode:ENiagaraSkeletalSpawning=::NewEnumerator1
   )
   &SkeletalMeshSkeletonLocation(Module.Inherit Velocity=false, Module.Use Sampled Bone as Ribbon ID=true)
   &ScaleColor(Scale RGB=&Multiply_VectorByFloat(Vector=<3 3 3>, Float=$User.LightMultiplier), ScaleA=true, ScaleRGB=true)
  ]
  rendererClasses [&NiagaraSpriteRendererProperties, &NiagaraLightRendererProperties]
  renderers [
   {rendererClass=&NiagaraSpriteRendererProperties}
   {rendererClass=&NiagaraLightRendererProperties, rendererIndex=1}
  ]
 }
 RT_Pink {
  emitterUpdateScript [
   &EmitterState(Life Cycle Mode=::NewEnumerator1, Loop Duration=$System.StaticLifetime, UseLoopDelay=false)
   &SpawnBurst_Instantaneous(Spawn Count=1, Use Spawn Probability=false)
   &SpawnRate(Use Spawn Probability=false) {enabled=false}
  ]
  particleSpawnScript [&InitializeParticle(
    Color=&MultiplyLinearColor(
     Linear Color=$User.LargeColor
     Scale Factor=&MakeCustomVector4FromBool(Bool:NiagaraBool=$User.bIsFrontEndPreview, True Vector4:Vector4f=<7 7 7 1>, False Vector4:Vector4f=<1 1 1 1>)
    )
    Lifetime=&Multiply_Float(A=$Emitter.CurrentLoopDuration, B=1)
    Mass=&UniformRangedFloat(Minimum=0.4, Maximum=2)
    Module.Write Color=true
    Module.Write Lifetime=true
    Module.Write Mass=false
    Module.Write Material Random=true
    Module.Write Position=false
    Module.Write Scale=false
    Module.Write SpriteRotation=true
    Module.Write SpriteSize=true
    Position=&SimulationPosition()
    Sprite Rotation=&NormalizedAngleToDegrees(NormalizedAngle=&UniformRangedFloat(Minimum=0, Maximum=1))
    Sprite Size=<0 0>
   )]
  particleUpdateScript [
   &ParticleState()
   &SolveForcesAndVelocity(Clamp Velocity=false, Limit Acceleration=false)
   &SampleSkeletalMeshSkeleton(
    Mesh=DataInterface() {propertyValues {
      DefaultMesh=null
      ExcludeBoneName=root
      OptionalComponentClass=null
      PreviewMesh=&F_MED_Rave_Ninja_Pack
      SkinningMode=SkinOnTheFly
      SoftSourceActor=null
      SourceMode=Default
      UvSetIndex=0
      WholeMeshLOD=-1
      bExcludeBone=false
      bReadDeformedGeometry=true
      bRequireCurrentFrameData=true
      ComponentTags []
      FilteredBones []
      FilteredSockets [LT_Big]
      MeshUserParameter {parameter {name=None, typeDefHandle {registeredTypeIndex=1}, varData []}}
      SamplingRegions []
     }}
    Spawn Mode:ENiagaraSkeletalSpawning=::NewEnumerator1
   )
   &SkeletalMeshSkeletonLocation(Module.Inherit Velocity=false, Module.Use Sampled Bone as Ribbon ID=true)
   &ScaleColor(
    Scale RGB=&VectorFromFloat(Value=&MakeFloatFromVector(
      Vector=&ScaleVectorByCurve(
       InitialValue:Vector3f=<1 1 1>
       ScaleCurve:NiagaraDataInterfaceVectorCurve=DataInterface() {propertyValues {
         XCurve {keys [
           {arriveTangent=2, interpMode=RCIM_Cubic, time=0, value=0.3}
           <at 0.5 1.3 RCIM_Cubic>
           {interpMode=RCIM_Cubic, leaveTangent=-2, time=1, value=0.3}
          ]}
         YCurve {keys [<at 0.5458 0 RCIM_Cubic>,<at 0.8822 0 RCIM_Cubic>,<at 1.2639 0 RCIM_Cubic>]}
         ZCurve {keys [<at 0.5458 0 RCIM_Cubic>,<at 0.8822 0 RCIM_Cubic>,<at 1.2639 0 RCIM_Cubic>]}
        }}
       CurveIndex:NiagaraFloat=$Particles.NormalizedAge
      )
      Channel=::NewEnumerator0
     ))
    ScaleA=true
    ScaleRGB=true
   )
  ]
  rendererClasses [&NiagaraSpriteRendererProperties, &NiagaraLightRendererProperties]
  renderers [
   {rendererClass=&NiagaraSpriteRendererProperties}
   {rendererClass=&NiagaraLightRendererProperties, rendererIndex=1}
  ]
 }
 RT_Pink001 {
  emitterUpdateScript [
   &EmitterState(Life Cycle Mode=::NewEnumerator1, Loop Behavior=::NewEnumerator1, Loop Duration Mode=::NewEnumerator1, UseLoopDelay=false)
   &SpawnBurst_Instantaneous(Spawn Count=1, Use Spawn Probability=false)
   &SpawnRate(Use Spawn Probability=false) {enabled=false}
  ]
  particleSpawnScript [&InitializeParticle(
    Color=$User.LargeColor
    Lifetime=&Multiply_Float(A=$Emitter.CurrentLoopDuration, B=1)
    Mass=&UniformRangedFloat(Minimum=0.4, Maximum=2)
    Module.Write Color=true
    Module.Write Lifetime=true
    Module.Write Mass=false
    Module.Write Material Random=true
    Module.Write Position=false
    Module.Write Scale=false
    Module.Write SpriteRotation=true
    Module.Write SpriteSize=true
    Position=&SimulationPosition()
    Sprite Rotation=&NormalizedAngleToDegrees(NormalizedAngle=&UniformRangedFloat(Minimum=0, Maximum=1))
    Sprite Size=<0 0>
   )]
  particleUpdateScript [
   &ParticleState(Kill Particles When Lifetime Has Elapsed=false, Loop Particles Lifetime=true)
   &SolveForcesAndVelocity(Clamp Velocity=false, Limit Acceleration=false)
   &SampleSkeletalMeshSkeleton(
    Mesh=DataInterface() {propertyValues {
      DefaultMesh=null
      ExcludeBoneName=root
      OptionalComponentClass=null
      PreviewMesh=&F_MED_Rave_Ninja_Pack
      SkinningMode=SkinOnTheFly
      SoftSourceActor=null
      SourceMode=Default
      UvSetIndex=0
      WholeMeshLOD=-1
      bExcludeBone=false
      bReadDeformedGeometry=true
      bRequireCurrentFrameData=true
      ComponentTags []
      FilteredBones []
      FilteredSockets [LT_Big]
      MeshUserParameter {parameter {name=None, typeDefHandle {registeredTypeIndex=1}, varData []}}
      SamplingRegions []
     }}
    Spawn Mode:ENiagaraSkeletalSpawning=::NewEnumerator1
   )
   &SkeletalMeshSkeletonLocation(Module.Inherit Velocity=false, Module.Use Sampled Bone as Ribbon ID=true)
   &ScaleColor(Scale RGB=&Multiply_VectorByFloat(Vector=<3 3 3>, Float=$User.LightMultiplier), ScaleA=true, ScaleRGB=true)
  ]
  rendererClasses [&NiagaraSpriteRendererProperties, &NiagaraLightRendererProperties]
  renderers [
   {rendererClass=&NiagaraSpriteRendererProperties}
   {rendererClass=&NiagaraLightRendererProperties, rendererIndex=1}
  ]
 }
 RT_Yellow {
  emitterUpdateScript [
   &EmitterState(
    Life Cycle Mode=::NewEnumerator1
    Loop Duration=&UniformRangedFloat(Minimum=0.5, Maximum=2)
    UseLoopDelay=false
   )
   &SpawnBurst_Instantaneous(Spawn Count=1, Spawn Probability=0.5, Use Spawn Probability=true)
   &SpawnRate(Use Spawn Probability=false) {enabled=false}
  ]
  particleSpawnScript [&InitializeParticle(
    Color=&MultiplyLinearColor(
     Linear Color=$User.SmallColor
     Scale Factor=&MakeCustomVector4FromBool(Bool:NiagaraBool=$User.bIsFrontEndPreview, True Vector4:Vector4f=<7 7 7 1>, False Vector4:Vector4f=<1 1 1 1>)
    )
    Lifetime=$Emitter.CurrentLoopDuration
    Mass=&UniformRangedFloat(Minimum=0.4, Maximum=2)
    Module.Write Color=true
    Module.Write Lifetime=true
    Module.Write Mass=false
    Module.Write Material Random=true
    Module.Write Position=false
    Module.Write Scale=false
    Module.Write SpriteRotation=true
    Module.Write SpriteSize=true
    Position=&SimulationPosition()
    Sprite Rotation=&NormalizedAngleToDegrees(NormalizedAngle=&UniformRangedFloat(Minimum=0, Maximum=1))
    Sprite Size=<0 0>
   )]
  particleUpdateScript [
   &ParticleState()
   &SolveForcesAndVelocity(Clamp Velocity=false, Limit Acceleration=false)
   &SampleSkeletalMeshSkeleton(
    Mesh=DataInterface() {propertyValues {
      DefaultMesh=null
      ExcludeBoneName=root
      OptionalComponentClass=null
      PreviewMesh=&F_MED_Rave_Ninja_Pack
      SkinningMode=SkinOnTheFly
      SoftSourceActor=null
      SourceMode=Default
      UvSetIndex=0
      WholeMeshLOD=-1
      bExcludeBone=false
      bReadDeformedGeometry=true
      bRequireCurrentFrameData=true
      ComponentTags []
      FilteredBones []
      FilteredSockets [RT_Small]
      MeshUserParameter {parameter {name=None, typeDefHandle {registeredTypeIndex=1}, varData []}}
      SamplingRegions []
     }}
    Spawn Mode:ENiagaraSkeletalSpawning=::NewEnumerator1
   )
   &SkeletalMeshSkeletonLocation(Module.Inherit Velocity=false, Module.Use Sampled Bone as Ribbon ID=true)
   &ScaleColor(
    Scale RGB=&VectorFromFloat(Value=&MakeFloatFromVector(
      Vector=&ScaleVectorByCurve(
       InitialValue:Vector3f=<1 1 1>
       ScaleCurve:NiagaraDataInterfaceVectorCurve=DataInterface() {propertyValues {
         XCurve {keys [
           {arriveTangent=2, interpMode=RCIM_Cubic, time=0, value=0}
           <at 0.5 1.3 RCIM_Cubic>
           {interpMode=RCIM_Cubic, leaveTangent=-2, time=1, value=0}
          ]}
         YCurve {keys [<at 0.5458 0 RCIM_Cubic>,<at 0.8822 0 RCIM_Cubic>,<at 1.2639 0 RCIM_Cubic>]}
         ZCurve {keys [<at 0.5458 0 RCIM_Cubic>,<at 0.8822 0 RCIM_Cubic>,<at 1.2639 0 RCIM_Cubic>]}
        }}
       CurveIndex:NiagaraFloat=$Particles.NormalizedAge
      )
      Channel=::NewEnumerator0
     ))
    ScaleA=true
    ScaleRGB=true
   )
  ]
  rendererClasses [&NiagaraSpriteRendererProperties, &NiagaraLightRendererProperties]
  renderers [
   {rendererClass=&NiagaraSpriteRendererProperties}
   {rendererClass=&NiagaraLightRendererProperties, rendererIndex=1}
  ]
 }
 RT_Yellow001 {
  emitterUpdateScript [
   &EmitterState(Life Cycle Mode=::NewEnumerator1, Loop Behavior=::NewEnumerator1, Loop Duration Mode=::NewEnumerator1, UseLoopDelay=false)
   &SpawnBurst_Instantaneous(Spawn Count=1, Spawn Probability=0.5, Use Spawn Probability=true)
   &SpawnRate(Use Spawn Probability=false) {enabled=false}
  ]
  particleSpawnScript [&InitializeParticle(
    Color=$User.SmallColor
    Lifetime=$Emitter.CurrentLoopDuration
    Mass=&UniformRangedFloat(Minimum=0.4, Maximum=2)
    Module.Write Color=true
    Module.Write Lifetime=true
    Module.Write Mass=false
    Module.Write Material Random=true
    Module.Write Position=false
    Module.Write Scale=false
    Module.Write SpriteRotation=true
    Module.Write SpriteSize=true
    Position=&SimulationPosition()
    Sprite Rotation=&NormalizedAngleToDegrees(NormalizedAngle=&UniformRangedFloat(Minimum=0, Maximum=1))
    Sprite Size=<0 0>
   )]
  particleUpdateScript [
   &ParticleState(Kill Particles When Lifetime Has Elapsed=false, Loop Particles Lifetime=true)
   &SolveForcesAndVelocity(Clamp Velocity=false, Limit Acceleration=false)
   &SampleSkeletalMeshSkeleton(
    Mesh=DataInterface() {propertyValues {
      DefaultMesh=null
      ExcludeBoneName=root
      OptionalComponentClass=null
      PreviewMesh=&F_MED_Rave_Ninja_Pack
      SkinningMode=SkinOnTheFly
      SoftSourceActor=null
      SourceMode=Default
      UvSetIndex=0
      WholeMeshLOD=-1
      bExcludeBone=false
      bReadDeformedGeometry=true
      bRequireCurrentFrameData=true
      ComponentTags []
      FilteredBones []
      FilteredSockets [RT_Small]
      MeshUserParameter {parameter {name=None, typeDefHandle {registeredTypeIndex=1}, varData []}}
      SamplingRegions []
     }}
    Spawn Mode:ENiagaraSkeletalSpawning=::NewEnumerator1
   )
   &SkeletalMeshSkeletonLocation(Module.Inherit Velocity=false, Module.Use Sampled Bone as Ribbon ID=true)
   &ScaleColor(Scale RGB=&Multiply_VectorByFloat(Vector=<3 3 3>, Float=$User.LightMultiplier), ScaleA=true, ScaleRGB=true)
  ]
  rendererClasses [&NiagaraSpriteRendererProperties, &NiagaraLightRendererProperties]
  renderers [
   {rendererClass=&NiagaraSpriteRendererProperties}
   {rendererClass=&NiagaraLightRendererProperties, rendererIndex=1}
  ]
 }
}
summary {
 systemName=NS_RaveNinja_Backpack_Lights
 emitters [
  {emitterName=LT_pink, rendererClasses [&NiagaraSpriteRendererProperties, &NiagaraLightRendererProperties]}
  {emitterName=RT_Pink, rendererClasses [&NiagaraSpriteRendererProperties, &NiagaraLightRendererProperties]}
  {emitterName=LT_Yellow, rendererClasses [&NiagaraSpriteRendererProperties, &NiagaraLightRendererProperties]}
  {emitterName=RT_Yellow, rendererClasses [&NiagaraSpriteRendererProperties, &NiagaraLightRendererProperties]}
  {emitterName=LT_pink001, rendererClasses [&NiagaraSpriteRendererProperties, &NiagaraLightRendererProperties]}
  {emitterName=RT_Pink001, rendererClasses [&NiagaraSpriteRendererProperties, &NiagaraLightRendererProperties]}
  {emitterName=LT_Yellow001, rendererClasses [&NiagaraSpriteRendererProperties, &NiagaraLightRendererProperties]}
  {emitterName=RT_Yellow001, rendererClasses [&NiagaraSpriteRendererProperties, &NiagaraLightRendererProperties]}
 ]
 userVariables [
  User.bIsFrontEndPreview {type=NiagaraBool, defaultValue=true}
  User.LargeColor {type=LinearColor, description="", defaultValue {struct=LinearColor, value=<rgba 1 0 0.6687 1>}}
  User.LightMultiplier {type=NiagaraFloat, defaultValue=0, description=""}
  User.SineSpeed {type=NiagaraFloat, defaultValue=0, description=""}
  User.SmallColor {type=LinearColor, description="", defaultValue {struct=LinearColor, value=<rgba 0.056 0.9425 1 1>}}
 ]
}
