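# Material function for the Rave Ninja backpack that adds an animated equalizer glow to incoming material attributes and overrides their base color, roughness, metallic and specular.
# UV1 selects between a procedural radial meter and a mirrored linear meter, both colored through a curve atlas.
# Fill level follows idle sine waves, switches to event audio amplitude and spectrum, and yields to a kill-react parameter.
# Debug readout and timer-based kill-react nodes are left unconnected.
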MaterialFunction /BRCosmetics/Accessories/FORT_Backpacks/Backpack_F_MED_Rave_Ninja/Materials/MF_RaveNinja
graph {nodes [
  n1 = CollectionParameter() {collection=&FortniteMaterialParameters, parameterName=Emote_AudioLow}
  n2 = CollectionParameter() {collection=&FortniteMaterialParameters, parameterName=Emote_AudioHigh}
  n3 = Multiply(
   A=TextureCoordinate()
   B=ComponentMask(None=VectorParameter() {group=RaveNinja, parameterName=RN_DebugTiling, defaultValue=<rgba 2 2 0 1>}.RGB) {g=true, r=true}
  ) {constB=2}
  n4 = ComponentMask(None=VectorParameter() {group=RaveNinja, parameterName=RN_DebugOffset, defaultValue=<rgba 0 0 0 1>}.RGB) {g=true, r=true}
  AppendVector(
   A=MaterialFunctionCall!(
    Vector2=AppendVector(A=@n1, B=@n2)
    MaximumNumberOfDigits
    UVs=Add(A=@n3, B=@n4)
    DebugTextLocation RG_UpperRight BA_LowerLeft
    Component Spacing
   ) {materialFunction=/Engine/Functions/Engine_MaterialFunctions02/Utility/DebugFloat2Values}.ColorCodedOutput
   B=Constant()
  )
  n5 = TextureCoordinate() {coordinateIndex=1}
  n6 = Constant2Vector() {g=2, r=1}
  n7 = VectorParameter() {group=Raveninja, parameterName=RN_Radial_Offset, defaultValue=<rgba 0 0 0 1>}
  n8 = Add(A=Add(A=Multiply(A=@n5, B=@n6), B=ComponentMask(None=@n7.RGB) {g=true, r=true})) {constB=-0.5}
  n9 = Arctangent2Fast(Y=ComponentMask(None=@n8) {g=true}, X=ComponentMask(None=@n8) {r=true})
  n10 = Multiply(A=@n8, B=@n8)
  n11 = ComponentMask(None=@n10) {r=true}
  n12 = ComponentMask(None=@n10) {g=true}
  n13 = AppendVector(A=Divide(A=ComponentMask(None=@n9) {r=true}), B=SquareRoot(None=Add(A=@n11, B=@n12)))
  n14 = Constant() {r=10}
  n15 = Multiply(A=Abs(None=ComponentMask(None=@n13) {r=true}), B=@n14) {constB=10}
  n16 = Add(A=@n15) {constB=-0.5}
  n17 = Distance(A=Sine(None=Multiply(A=@n16) {constB=0.5}), B=Constant())
  n18 = Divide(A=Ceil(None=@n15), B=Add(A=@n14, B=Multiply(A=@n14) {constB=0.5}) {constB=5})
  n19 = LinearInterpolate(
   A=ScalarParameter() {defaultValue=0.9, group=Raveninja, parameterName=RN_RadialCellWidth_Max}
   B=ScalarParameter() {defaultValue=0.9, group=Raveninja, parameterName=RN_RadialCellWidth_Min}
   Alpha=@n18
  )
  n20 = Ceil(None=OneMinus(None=Divide(A=@n17, B=@n19)))
  Clamp(None=@n20)
  n21 = Time()
  n22 = ScalarParameter() {defaultValue=-1000000000, parameterName=RN_GameTimeFromBP}
  n23 = Saturate(None=Divide(A=Subtract(A=@n21, B=@n22), B=ScalarParameter() {defaultValue=3.3375, parameterName=RM_KillReactDuration}))
  ConstantBiasScale(None=Sine(None=Add(A=@n23) {constB=-0.25}))
  CurveAtlasRowParameter(CurveTime=@n23) {
   atlas=&AthenaCurveAtlas01
   curve=/Game/Characters/Enemies/Cattus_Head/Materials/Cattus_EyeMovement_Curve
   defaultValue=92
   parameterName=KillReactCurve
  }
  Frac(None=Multiply(A=TextureCoordinate()) {constB=2})
  n24 = GetMaterialAttributes(None=FunctionInput() {inputName=Material, inputType=FunctionInput_MaterialAttributes}) {attributeGetTypes [#1]}
  n25 = VectorParameter() {group=RaveNinja, parameterName=RN_BGColor, defaultValue=<rgba 0.6939 1 0.2384 1>}
  n26 = Multiply(A=@n25.RGB, B=ScalarParameter() {defaultValue=1, group=RaveNinja, parameterName=RN_BGColor_Strength})
  n27 = Time()
  n28 = ScalarParameter() {defaultValue=0.35, group=RaveNinja, parameterName=RN_DefaultSineWaveSpeed}
  n29 = Multiply(A=@n27, B=@n28)
  n30 = Divide(A=ConstantBiasScale(None=Sine(None=@n29)), B=ScalarParameter() {defaultValue=1, group=RaveNinja, parameterName=RN_DefaultAmplitudeBias})
  n31 = Constant() {r=0.2}
  n32 = Sine(None=Multiply(A=Add(A=@n27) {constB=0.25}, B=Multiply(A=@n31, B=@n28)))
  n33 = Add(A=@n27) {constB=0.5}
  n34 = Multiply(A=Constant() {r=0.5}, B=@n28)
  n35 = ConstantBiasScale(None=Sine(None=Multiply(A=@n33, B=@n34) {constB=0.5}))
  n36 = ConstantBiasScale(None=Sine(None=Multiply(A=@n27, B=@n28)))
  n37 = Divide(A=Add(A=ConstantBiasScale(None=@n32), B=Add(A=@n35, B=@n36))) {constA=3, constB=3}
  n38 = CollectionParameter() {collection=&FortniteMaterialParameters, parameterName=EventAudio_Amplitude}
  n39 = CollectionParameter() {collection=&FortniteMaterialParameters, parameterName=EventAudio_Spectrum}
  n40 = MaterialFunctionCall(Float4=@n39) {materialFunction=&BreakOutFloat4Components}
  n41 = Add(A=Add(A=@n40.R, B=@n40.R), B=Add(A=@n40.B, B=@n40.B))
  n42 = Saturate(None=Multiply(A=@n38, B=Divide(A=@n41)) {constB=0.5})
  n43 = Clamp(None=Ceil(None=@n42))
  n44 = OneMinus(None=SphereMask(
    A=@n42
    B=ScalarParameter() {defaultValue=0.01, group=Raveninja, parameterName=RN_EmoteThreshold}
    Radius=Constant() {r=0.1}
    Hardness=Constant() {r=1}
   ))
  n45 = ScalarParameter() {parameterName=RN_KillReactCurve}
  n46 = OneMinus(None=@n45)
  n47 = Clamp(None=Ceil(None=@n45))
  n48 = LinearInterpolate(
   A=LinearInterpolate(A=LinearInterpolate(A=@n30, B=@n37, Alpha=@n43), B=OneMinus(None=@n42), Alpha=@n44)
   B=Clamp(None=Multiply(A=@n46) {constB=2})
   Alpha=@n47
  )
  n49 = Multiply(A=@n48) {constB=-1}
  n50 = CurveAtlasRowParameter(CurveTime=Add(A=@n18, B=@n49)) {
   atlas=&AthenaCurveAtlas01
   curve=/Game/Accessories/FORT_Backpacks/Backpack_F_MED_Rave_Ninja/Textures/T_RaveNinjaColors_01
   defaultValue=209
   group=Raveninja
   parameterName=RN_ColorCurve
  }
  n51 = Subtract(A=@n18, B=@n48) {constB=-0.1}
  n52 = Divide(A=Constant() {r=1}, B=@n14)
  n53 = Multiply(A=@n20, B=Ceil(None=Subtract(A=@n51, B=@n52) {constB=0.1}))
  n54 = Clamp(None=@n53)
  n55 = ScalarParameter() {defaultValue=1, parameterName=RN_EQ_DefaultGlow_Strength}
  n56 = ScalarParameter() {defaultValue=5, parameterName=RN_EQ_ReactGlow_Strength}
  n57 = Max(A=@n47, B=@n44)
  n58 = Add(
   A=TextureCoordinate() {coordinateIndex=1}
   B=ComponentMask(None=VectorParameter() {group=Raveninja, parameterName=RN_Equalizer_Position, defaultValue=<rgba 0 0.02 0 1>}.RGB) {g=true, r=true}
  )
  n59 = MaterialFunctionCall(Float2=Frac(None=Multiply(A=@n58) {constB=2})) {materialFunction=&BreakOutFloat2Components}
  n60 = ChannelMaskParameter(None=MaterialFunctionCall(X=@n59.R, Y=OneMinus(None=@n59.G)) {materialFunction=&MakeFloat2}.Result) {group=Raveninja, maskChannel=Green, parameterName=RN_EquilizerDirectionChannel, defaultValue=<rgba 0 1 0 0>}
  n61 = ChannelMaskParameter(None=@n58) {group=Raveninja, maskChannel=Green, parameterName=RN_EquilizerDirectionChannel, defaultValue=<rgba 0 1 0 0>}
  n62 = LinearInterpolate(A=@n60.R, B=OneMinus(None=@n60.R), Alpha=Round(None=@n61.R))
  n63 = ScalarParameter() {defaultValue=10, group=Raveninja, parameterName=RN_EqualizerLevels}
  n64 = Multiply(A=@n62, B=Multiply(A=@n63, B=Multiply(A=@n62, B=@n62)))
  n65 = Divide(A=Ceil(None=@n64), B=@n63)
  n66 = Add(A=@n65, B=@n49)
  n67 = CurveAtlasRowParameter(CurveTime=@n66) {
   atlas=&AthenaCurveAtlas01
   curve=&T_RaveNinjaColors_01
   defaultValue=209
   group=Raveninja
   parameterName=RN_ColorCurve
  }
  n68 = Multiply(A=@n67, B=Multiply(A=LinearInterpolate(A=@n55, B=@n56, Alpha=@n57), B=OneMinus(None=@n66)))
  n69 = Sine(None=Multiply(A=Add(A=@n64) {constB=-0.5}) {constB=0.5})
  n70 = Constant()
  n71 = ScalarParameter() {defaultValue=0.9, group=Raveninja, parameterName=RN_CellWidth_Max}
  n72 = ScalarParameter() {defaultValue=0.9, group=Raveninja, parameterName=RN_CellWidth_Min}
  n73 = Multiply(A=@n65, B=@n65)
  n74 = OneMinus(None=Divide(A=Distance(A=@n69, B=@n70), B=LinearInterpolate(A=@n71, B=@n72, Alpha=@n73)))
  n75 = Multiply(A=Clamp(None=Ceil(None=@n74)), B=Ceil(None=Subtract(A=@n65, B=@n48)))
  n76 = ComponentMask(None=TextureCoordinate() {coordinateIndex=1}) {g=true}
  n77 = Multiply(
   A=LinearInterpolate(A=LinearInterpolate(A=@n26, B=@n50, Alpha=@n54), B=LinearInterpolate(A=@n26, B=@n68, Alpha=@n75), Alpha=Round(None=@n76))
   B=ScalarParameter() {defaultValue=1, group=Raveninja, parameterName=RN_Equalizer_Glow_Strength}
  )
  n78 = Constant()
  n79 = SetMaterialAttributes!(
   MaterialAttributes=@n24.MaterialAttributes
   EmissiveColor=Add(A=@n24.EmissiveColor, B=Power(Base=@n77))
   Roughness=ScalarParameter() {defaultValue=1, parameterName=RN_Roughness}
   Specular=@n78
   BaseColor=@n25.RGB
   Metallic=@n78
  ) {attributeSetTypes [
    #1
    #2
    #3
    #4
    #5
   ]}
  FunctionOutput(None=@n79)
  n80 = CollectionParameter() {collection=&FortniteMaterialParameters, parameterName=Emote_AudioLow}
  n81 = CollectionParameter() {collection=&FortniteMaterialParameters, parameterName=Emote_AudioHigh}
  Max(A=Add(A=@n80, B=@n81))
  Multiply(A=@n50, B=@n53)
 ]}
