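# Master material for the Rave Ninja backpack cosmetic, used on skeletal meshes with mesh deformer support.
# The shared character PreFX function feeds its material attributes through MF_RaveNinja into the character PostFX function.
# The PreFX AO output drives PostFX's AO, Distance Blend and Fresnel inputs alike.
# Shading model comes from the material expression, with Default Lit and Subsurface Profile in use.
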Material /BRCosmetics/Accessories/FORT_Backpacks/Backpack_F_MED_Rave_Ninja/Materials/F_MED_Rave_Ninja_Backpack_MASTER
graph {
 properties {
  bUseMaterialAttributes=true
  bUsedWithMeshDeformer=true
  bUsedWithSkeletalMesh=true
  pixelDepthOffsetMode=PDOM_Legacy
  shadingModel=MSM_FromMaterialExpression
  usedShadingModels=Default Lit | Subsurface Profile
 }
 nodes [
  n1 = MaterialFunctionCall() {materialFunction=&BaseFN_CharacterFunction_PreFX}
  MaterialFunctionCall(
   AO=@n1.AO
   Distance Blend=@n1.AO
   Fresnel=@n1.AO
   Material=MaterialFunctionCall!(Material=@n1.Result) {materialFunction=&/MF_RaveNinja}.Result
  ) {materialFunction=&BaseFN_CharacterFunction_PostFX}
 ]
}
