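# Animation Blueprint for the Rave Ninja back bling, a turntable pack that reacts to the wearer's kills and movement.
# Bone modifiers spin the record, rotate the tone arm, shift sliders on jump or crouch, and offset the pack by skydiving state and spine bend.
# A watched kill-stat increase swings the arm and plays a record-scratch montage, applied additively over the cached procedural pose until a delay clears the flag.
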AnimBlueprint /BRCosmetics/Accessories/FORT_Backpacks/Backpack_F_MED_Rave_Ninja/Meshes/F_MED_Rave_Ninja_Pack_AnimBp
target_skeleton=&/F_MED_Rave_Ninja_Pack_Skeleton
graphs [
 AnimGraph {nodes [
   n1 = Apply Additive(Base=Use cached pose 'RecordCache'(), Additive=Slot 'DefaultSlot'(), Alpha=2) {alphaBoolBlend {blendOption=Linear}, alphaScaleBiasClamp {bInterpResult=true, interpSpeedDecreasing=4, interpSpeedIncreasing=4}}
   Output Pose(Result=Blend Poses by bool(
     BlendPose_0=@n1
     BlendPose_1=Use cached pose 'RecordCache'()
     BlendTime_0=0
     BlendTime_1=0
     bActiveValue=Get Active_Kill() -> (Active_Kill:bool)
    ) {blendType=Linear}) {layerGroup=None}
   n2 = "Transform (Modify) Bone - Bone: root"(ComponentPose=Local To Component(LocalPose=Input Pose()), Translation=Get backpack_translation(), Rotation=Get backpack_rotate()) {
    rotationMode=BMM_Additive
    rotationSpace=BCS_BoneSpace
    translationMode=BMM_Additive
    translationSpace=BCS_BoneSpace
    alphaBoolBlend {blendOption=Linear}
    boneToModify {boneName=root}
   }
   n3 = "Transform (Modify) Bone - Bone: pack"!(ComponentPose=@n2, bAlphaBoolEnabled:bool, Translation="0,-2,0", Rotation="0,0,0", Scale) -> (Pose) {alphaInputType=Bool, translationMode=BMM_Additive, alphaBoolBlend {blendOption=Linear}, boneToModify {boneName=pack}, translation {y=-2}}
   n4 = "Transform (Modify) Bone - Bone: pack"!(ComponentPose=@n3, Alpha:float=Get skydiving() -> (skydiving:double), Translation="0,-0.25,0", Rotation="0,0,0", Scale) -> (Pose) {
    rotationMode=BMM_Additive
    translationMode=BMM_Additive
    alphaBoolBlend {blendOption=Linear}
    alphaScaleBiasClamp {bInterpResult=true, interpSpeedDecreasing=7.5, interpSpeedIncreasing=7.5}
    boneToModify {boneName=pack}
    translation {y=-0.25}
   }
   n5 = Get JumpCrouch_Alpha() -> (JumpCrouch_Alpha:double)
   n6 = "Transform (Modify) Bone - Bone: dyn_slider_02"(ComponentPose=@n4, Alpha=@n5, Translation="0,-0.6,4", Rotation="0,0,0", Scale="1,1,1") {
    translationMode=BMM_Additive
    translationSpace=BCS_BoneSpace
    alphaBoolBlend {blendOption=Linear}
    alphaScaleBiasClamp {bInterpResult=true, interpSpeedDecreasing=2, interpSpeedIncreasing=2}
    boneToModify {boneName=dyn_slider_02}
    translation {y=-0.6, z=4}
   }
   n7 = "Transform (Modify) Bone - Bone: dyn_slider_01"(ComponentPose=@n6, Alpha=@n5, Translation="0,0,-4", Rotation="0,0,0", Scale="1,1,1") {
    translationMode=BMM_Additive
    translationSpace=BCS_BoneSpace
    alphaBoolBlend {blendOption=Linear}
    alphaScaleBiasClamp {bInterpResult=true, interpSpeedDecreasing=4, interpSpeedIncreasing=6}
    boneToModify {boneName=dyn_slider_01}
    translation {z=-4}
   }
   n8 = "Transform (Modify) Bone - Bone: dyn_tonearm_parent"(ComponentPose=@n7, Alpha=1, Translation="0,0,0", Rotation=Get ToneArm() -> (ToneArm:Rotator), Scale="1,1,1") {rotationMode=BMM_Additive, rotationSpace=BCS_BoneSpace, alphaBoolBlend {blendOption=Linear}, boneToModify {boneName=dyn_tonearm_parent}}
   n9 = "Transform (Modify) Bone - Bone: dyn_record_parent"(ComponentPose=@n8, Alpha=1, Translation="0,0,0", Rotation=Get SpinnerRot() -> (SpinnerRot:Rotator), Scale="1,1,1") {
    rotationMode=BMM_Replace
    rotationSpace=BCS_BoneSpace
    alphaBoolBlend {blendOption=Linear}
    alphaScaleBiasClamp {bInterpResult=true, interpSpeedDecreasing=5, interpSpeedIncreasing=6}
    boneToModify {boneName=dyn_record_parent}
   }
   Save cached pose 'RecordCache'(Pose=Component To Local(ComponentPose=@n9)) {cacheName=RecordCache}
  ]}
 EventGraph {nodes [
   Event BlueprintInitializeAnimation()
   > Set Skel_Mesh(Skel_Mesh=GetOwnerSkeletalMeshComponent())
   > n1 = Cast To FortPlayerPawn(Object=TryGetFortPawn())
   > Set Pawn(Pawn:Fort Player Pawn Object Reference=@n1) -> (Output_Get:Fort Player Pawn Object Reference)
   Event OnCharacterCustomizationLoaded()
   > Set ToneArm_Moving(ToneArm_Moving:bool=true) -> (Output_Get:bool)
   > Set ToneArm_DirectionTowardsRecord(ToneArm_DirectionTowardsRecord:bool=false) -> (Output_Get:bool)
   > Set ToneArm(ToneArm:Rotator=MakeRotator(Roll=Get End_const() -> (End_const:double))) -> (Output_Get:Rotator)
   n2 = Event BlueprintUpdateAnimation()
   > Get!(execute:Exec) -> (then:Exec, else:Exec, Skel_Mesh:Skeletal Mesh Component Object Reference)
   > Get!(execute) -> (then, else, Pawn:Fort Player Pawn Object Reference)
   > n3 = Sequence!(execute) -> (then_0, then_1, then_2:Exec, then_3:Exec, then_4:Exec, then_5:Exec, then_6:Exec, then_7:Exec)
   Reroute Node(InputPin=@n3.then_3)
   n4 = NOT Boolean(A=Get bIsParachuteOpened())
   n5 = Get bIsSkyDiving()
   Set skydiving(execute=@n3.then_7, skydiving:double="To Float (Boolean)"(InBool=AND Boolean(A=@n5, B=@n4))) -> (Output_Get:double)
   Branch(execute=@n3.then_2, Condition=Get Active_Kill())
   > n6 = Sequence!(execute) -> (then_0, then_1)
   Do Once(execute=@n6.then_1, Reset=@n9.then)
   > n7 = Montage_Play!(
    execute
    self:Anim Instance Object Reference
    MontageToPlay:Anim Montage Object Reference=&/F_MED_Rave_Ninja_RecordScratch_Montage
    InPlayRate:float=1
    ReturnValueType:EMontagePlayReturnType Enum=MontageLength
    InTimeToStartMontageAt:float=0
    bStopAllMontages:bool=false
   ) -> (then, ReturnValue:float)
   n8 = Get End_const()
   Set ToneArm(
    execute=@n6.then_0
    ToneArm=RInterpTo!(Current:Rotator=Get ToneArm(), Target:Rotator=MakeRotator(Roll=@n8), DeltaTime:float=@n2.DeltaTimeX, InterpSpeed:float=15) -> (ReturnValue:Rotator)
   )
   > Set ToneArm_Moving(ToneArm_Moving=true)
   > Set ToneArm_DirectionTowardsRecord(ToneArm_DirectionTowardsRecord=true)
   > Delay(Duration=@n7)
   > n9 = Set Active_Kill(Active_Kill:bool=false) -> (Output_Get:bool)
   n10 = float / float(A=GetRealTimeSeconds(), B=-720) {is=title K2Node_CallFunction}
   Set SpinnerRot(execute=@n3.then_1, SpinnerRot:Rotator=MakeRotator(Roll=float * float(A=@n10, B=100000))) -> (Output_Get:Rotator)
   n11 = Get bIsJumping()
   n12 = Get bIsCrouching()
   Set JumpCrouch_Alpha(execute=@n3.then_5, JumpCrouch_Alpha:double=SelectFloat(A=1, B=0, bPickA=OR Boolean(A=@n11, B=@n12))) -> (Output_Get:double)
   n13 = GetOwnerSkeletalMeshComponent()
   n14 = BreakRotator(InRot=BreakTransform(InTransform=GetDeltaTransformFromRefPose(self=@n13, BoneName=spine_05, BaseName=spine_01)).Rotation)
   Set backpack_rotate(execute=@n3.then_6, backpack_rotate=MakeRotator(Roll=MapRangeClamped(Value=@n14.Yaw, InRangeA=0, InRangeB=-30, OutRangeB=20)))
   > Set backpack_translation(backpack_translation=MakeVector(Y=MapRangeClamped(Value=@n14.Yaw, InRangeA=0, InRangeB=-30, OutRangeB=-3)))
   n15 = WatchCosmeticStat!(
    execute:Exec=@n3.then_0
    Pawn:Fort Player Pawn Object Reference=Get Pawn() -> (Pawn:Fort Player Pawn Object Reference)
    StatTag:Fort Stat Manager Tag Structure=(TagName="StatManager.AthenaKills")
    DefaultValue:int=0
    bBroadcastInitialValue:bool=true
    InitialValueBroadcastDelay:float=-1
    ChangeValueBroadcastDelay:float=-1
    bConsumeWaitingNotifications:bool=true
   ) -> (then:Exec, OnStatChanged:Exec, StatName:Name, StatValue:int) {is=title K2Node_AsyncAction}
   Set Did_Kill(execute=@n15.OnStatChanged, Did_Kill:bool=false) -> (Output_Get:bool)
   > n16 = Set KillCount(KillCount:double="To Float (Integer)"(InInt=@n15.StatValue)) -> (Output_Get:double)
   > Branch(Condition="float > float"(A=@n16, B=Get Last_KillCount() -> (Last_KillCount:double)) {is=title K2Node_CallFunction})
   > Set Active_Kill(Active_Kill=true)
   > Set Last_KillCount(Last_KillCount:double=@n16) -> (Output_Get:double)
   n17 = Get Second!(A:Date Time Structure=Now!() -> (ReturnValue:Date Time Structure)) -> (ReturnValue:int=0) {is=title K2Node_CallFunction}
   Branch(Condition="Not Equal (Integer)"!(A:int=@n17, B:int=Get Timer() -> (Timer:int=0)) -> (ReturnValue:bool) {is=title K2Node_CallFunction})
   > n18 = Set Timer(Timer:int=@n17) -> (Output_Get:int=0)
   > Branch(Condition="Equal (Integer)"(A=0, B="% (Integer)"(A=@n18, B=10) -> (ReturnValue=0)) {is=title K2Node_CallFunction})
   > n19 = Branch(Condition=Get ToneArm_Moving() -> (ToneArm_Moving:bool))
   > Set ToneArm_Moving(ToneArm_Moving=false)
   > n20 = Branch(Condition=Get ToneArm_DirectionTowardsRecord() -> (ToneArm_DirectionTowardsRecord:bool))
   > n21 = Set ToneArm_DirectionTowardsRecord(ToneArm_DirectionTowardsRecord=false)
   Gate!(Enter:Exec, Open:Exec=[@n23.then, @n21.then], Close:Exec=@n24.then, Toggle:Exec, bStartClosed:bool=true) -> (Exit:Exec) {is=title K2Node_MacroInstance}
   n22 = Select(Option 0:double=Get Start_const() -> (Start_const:double), Option 1:double=Get End_const(), Index=Get ToneArm_DirectionTowardsRecord()) -> (ReturnValue:double)
   > Set ToneArm(ToneArm=RInterpTo(Current=Get ToneArm(), Target=MakeRotator(Roll=@n22), DeltaTime=0, InterpSpeed=1))
   n23 = Set ToneArm_DirectionTowardsRecord(execute=@n20.else, ToneArm_DirectionTowardsRecord=true)
   n24 = Set ToneArm_Moving(execute=@n19.else, ToneArm_Moving=true)
   TryGetPawnOwner()
   Reroute Node(InputPin:float=@n2.DeltaTimeX) -> (OutputPin:float)
   Get Start_const()
  ]}
]
variables [
 Active_Kill {type=boolean}
 backpack_rotate {type=Rotator}
 backpack_translation {type=Vector}
 Did_Kill {type=boolean}
 End_const {type=number, default=-15}
 JumpCrouch_Alpha {type=number}
 KillCount {type=number}
 Last_KillCount {type=number}
 Pawn {type=FortPlayerPawn}
 Skel_Mesh {type=SkeletalMeshComponent}
 skydiving {type=number}
 SpinnerRot {type=Rotator}
 Start_const {type=number, default=15}
 Timer {type=integer}
 ToneArm {type=Rotator}
 ToneArm_DirectionTowardsRecord {type=boolean}
 ToneArm_Moving {type=boolean}
]
